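; Child Blueprint for the Neon City sushi table A environment prop, with no gameplay logic of its own.
; Its construction script and AddNewInstanceToComponent override, which takes an instanced static mesh component and a transform, only forward to the parent class.
; BeginPlay, ActorBeginOverlap and Tick are disabled placeholder nodes, so any instancing behavior comes from the parent, which is not shown here.
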(event EventBeginPlay
  (|CallParentFunction) ; This node is disabled and will not be called.\nDrag off pins to build functionality.
)

(event Collision|EventActorBeginOverlap
  (params (OtherActor :type Actor%20Object%20Reference))
  (|CallParentFunction OtherActor) ; This node is disabled and will not be called.\nDrag off pins to build functionality.
)

(event EventTick
  (params (DeltaSeconds :type Float%20%28single-precision%29))
  (|CallParentFunction DeltaSeconds) ; This node is disabled and will not be called.\nDrag off pins to build functionality.
)

(fn AddNewInstanceToComponent
  (params (comp :type Instanced%20Static%20Mesh%20Component%20Object%20Reference) (InTransform :type Transform%20%28by%20ref%29))
  (|CallParentFunction comp InTransform))

(fn ConstructionScript
  (|CallParentFunction))
